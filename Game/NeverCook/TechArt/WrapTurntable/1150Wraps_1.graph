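# Level Sequence in the WrapTurntable tech-art folder that stages a shot of a wrap display mesh, apparently for rendering wrap cosmetics.
# Camera_Wrap holds constant transform, focal length, focus and aperture under a camera cut track, and the Wrap actor holds a fixed pose with an empty animation track.
# Despite the turntable name, only three spotlights vary, with intensities keyed at 0 through 9, suggesting per-frame lighting setups.
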LevelSequence /Game/NeverCook/TechArt/WrapTurntable/1150Wraps_1
bindings [
 lgt_key2 {id=#1, possessed_class=SpotLight, tracks []}
 CameraComponent {
  id=#2
  parent_id=#3
  possessed_class=CineCameraComponent
  tracks [
   {
    type=MovieSceneFloatTrack
    display_name=Current Focal Length
    sections [{type=MovieSceneFloatSection, channels [None {type=MovieSceneScriptingFloatChannel, default=20, const [20,1]}]}]
   }
   {
    type=MovieSceneFloatTrack
    display_name="Manual Focus Distance (Focus Settings)"
    sections [{type=MovieSceneFloatSection, channels [None {type=MovieSceneScriptingFloatChannel, default=48.1345, const [48.1345,1]}]}]
   }
   {
    type=MovieSceneFloatTrack
    display_name=Current Aperture
    sections [{type=MovieSceneFloatSection, channels [None {type=MovieSceneScriptingFloatChannel, default=22, const [22,1]}]}]
   }
  ]
 }
 LightComponent0 {
  id=#4
  parent_id=#5
  possessed_class=SpotLightComponent
  tracks [
   {
    type=MovieSceneFloatTrack
    display_name=Intensity
    sections [{
      type=MovieSceneFloatSection
      channels [None {type=MovieSceneScriptingFloatChannel, default=3, thinned_from=10, keys [[0,25],[1,15],[2,20],[3,3],[4,15],[5,0],[6,8],[8,8],[9,6]]}]
     }]
   }
   {
    type=MovieSceneColorTrack
    display_name=Light Color
    sections [{
      type=MovieSceneColorSection
      channels [
       Color.R {type=MovieSceneScriptingFloatChannel, default=1}
       Color.G {type=MovieSceneScriptingFloatChannel, default=1}
       Color.B {type=MovieSceneScriptingFloatChannel, default=1}
       Color.A {type=MovieSceneScriptingFloatChannel, default=1}
      ]
     }]
   }
  ]
 }
 LightComponent0 {
  id=#6
  parent_id=#1
  possessed_class=SpotLightComponent
  tracks [
   {
    type=MovieSceneFloatTrack
    display_name=Intensity
    sections [{
      type=MovieSceneFloatSection
      channels [None {type=MovieSceneScriptingFloatChannel, default=10, thinned_from=10, keys [[0,8],[1,5],[2,5],[3,10],[4,5],[8,5],[9,4]]}]
     }]
   }
   {
    type=MovieSceneColorTrack
    display_name=Light Color
    sections [{
      type=MovieSceneColorSection
      channels [
       Color.R {type=MovieSceneScriptingFloatChannel, default=1}
       Color.G {type=MovieSceneScriptingFloatChannel, default=1}
       Color.B {type=MovieSceneScriptingFloatChannel, default=1}
       Color.A {type=MovieSceneScriptingFloatChannel, default=1}
      ]
     }]
   }
  ]
 }
 LightComponent0 {
  id=#7
  parent_id=#8
  possessed_class=SpotLightComponent
  tracks [
   {
    type=MovieSceneFloatTrack
    display_name=Intensity
    sections [{
      type=MovieSceneFloatSection
      channels [None {type=MovieSceneScriptingFloatChannel, default=4, keys [[0,5],[1,5],[2,8],[3,4],[4,10],[5,2],[6,3],[7,3],[8,4],[9,4]]}]
     }]
   }
   {
    type=MovieSceneColorTrack
    display_name=Light Color
    sections [{
      type=MovieSceneColorSection
      channels [
       Color.R {type=MovieSceneScriptingFloatChannel, default=1}
       Color.G {type=MovieSceneScriptingFloatChannel, default=1}
       Color.B {type=MovieSceneScriptingFloatChannel, default=1}
       Color.A {type=MovieSceneScriptingFloatChannel, default=1}
      ]
     }]
   }
  ]
 }
 lgt_key3 {id=#8, possessed_class=SpotLight, tracks []}
 Camera_Wrap {
  id=#3
  possessed_class=CineCameraActor
  tracks [{
    type=MovieScene3DTransformTrack
    sections [{
      type=MovieScene3DTransformSection
      channels [
       Location.X {type=MovieSceneScriptingDoubleChannel, default=-15.1117, const [-15.1117,1]}
       Location.Y {type=MovieSceneScriptingDoubleChannel, default=-43.7456, const [-43.7456,1]}
       Location.Z {type=MovieSceneScriptingDoubleChannel, default=12.0619, const [12.0619,1]}
       Rotation.X {type=MovieSceneScriptingDoubleChannel, default=5.4412, const [5.4412,1]}
       Rotation.Y {type=MovieSceneScriptingDoubleChannel, default=-0.1999, const [-0.1999,1]}
       Rotation.Z {type=MovieSceneScriptingDoubleChannel, default=65.7957, const [65.7957,1]}
       Scale.X {type=MovieSceneScriptingDoubleChannel, default=1, const [1,1]}
       Scale.Y {type=MovieSceneScriptingDoubleChannel, default=1, const [1,1]}
       Scale.Z {type=MovieSceneScriptingDoubleChannel, default=1, const [1,1]}
      ]
     }]
   }]
 }
 lgt_key {id=#5, possessed_class=SpotLight, tracks []}
 Wrap {
  id=#9
  possessed_class=FortWrapSkeletalMeshActor
  tracks [
   {
    type=MovieScene3DTransformTrack
    sections [{
      type=MovieScene3DTransformSection
      channels [
       Location.X {type=MovieSceneScriptingDoubleChannel, default=49.8873, const [49.8873,1]}
       Location.Y {type=MovieSceneScriptingDoubleChannel, default=82.0293, const [82.0293,1]}
       Location.Z {type=MovieSceneScriptingDoubleChannel, default=0.7688, const [0.7688,1]}
       Rotation.X {type=MovieSceneScriptingDoubleChannel, default=13.6396, const [13.6396,1]}
       Rotation.Y {type=MovieSceneScriptingDoubleChannel, default=53.1996, const [53.1996,1]}
       Rotation.Z {type=MovieSceneScriptingDoubleChannel, default=43.5747, const [43.5747,1]}
       Scale.X {type=MovieSceneScriptingDoubleChannel, default=1, const [1,1]}
       Scale.Y {type=MovieSceneScriptingDoubleChannel, default=1, const [1,1]}
       Scale.Z {type=MovieSceneScriptingDoubleChannel, default=1, const [1,1]}
      ]
     }]
   }
   {type=MovieSceneSkeletalAnimationTrack, sections []}
  ]
 }
]
display_rate {denominator=1, numerator=60}
tick_resolution {denominator=1, numerator=60000}
tracks [{type=MovieSceneCameraCutTrack, sections [{type=MovieSceneCameraCutSection, end_frame=300, start_frame=0}]}]
