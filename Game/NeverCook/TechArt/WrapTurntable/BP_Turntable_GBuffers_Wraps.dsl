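; Turntable actor for reviewing assets, apparently wraps, by switching between lit and G-buffer or material debug passes.
; Pass functions feed SetPP for viewport or MRQ_PostProcessSwitch for Movie Render Queue, toggling post-process components, show flags and TAA console variables.
; A UI widget shows the pass and a mesh list with vertex counts read from asset tags.
; Helpers configure inset shadows, a BaseColor levels-checker material and Time of Day Manager visibility and rotation.
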(event EventBeginPlay
  (bind _returnvalue (UserInterface|CreateWidget "/Gbuffer_Turntable/FN_PublicTurntable_Assets/BP/WBP_Turntable_UI.WBP_Turntable_UI_C"))
  (bind _output_get (Variables|Default|SetUI _returnvalue))
  (UserInterface|Viewport|AddtoViewport _output_get 0)
  (Class|WBPTurntableUI|SetText (Variables|Default|GetUI) (Variables|UIValues|GetuserName) (Variables|UIValues|Getdate) (Variables|UIValues|Getbuild) (Variables|UIValues|GetobjectList) "lit" "0001" (Variables|UIValues|GetScalability)))

(event Collision|EventActorBeginOverlap
  (params (OtherActor :type Actor%20Object%20Reference)))

(event Custom|SetTheCurrentFrameNumber
  (params (FrameNumber :type Integer)))

(event EventTick
  (params (DeltaSeconds :type Float%20%28single-precision%29)))

(event EventEndPlay
  (params (EndPlayReason :type EEndPlayReason%20Enum :default "Destroyed"))
  (Development|ExecuteConsoleCommand "r.TemporalAASamples 8")
  (CallFunction|LitMode))

(event Custom|LevelChange
  (params (Pass :type String)))

(fn baseColor
  (CallFunction|SetPP false true))

(fn checkVertexCount
  (for _array_element (Components|GetChildrenComponents (Variables|/Script/Engine.Actor|GetRootComponent (Variables|Default|GetTurntable_Objects)) false)
    (bind _asstatic_mesh_component (Utilities|Casting|CastToStaticMeshComponent _array_element))
    (bind _returnvalue (CallFunction|VertexCount (Variables|/Script/Engine.StaticMeshComponent|GetStaticMesh _asstatic_mesh_component)))
    (Development|PrintString (Utilities|String|Append "staticMesh " (Utilities|GetDisplayName (Variables|/Script/Engine.StaticMeshComponent|GetStaticMesh _asstatic_mesh_component)) " " (Utilities|String|ToString(Integer) _returnvalue)))))

(fn DEBUG_CheckForDumpFrame
  (return (Variables|Default|GetUseDumpGPU?) (Variables|Default|GetDumpFrame)))

(fn emissive
  (CallFunction|SetPP false false false false false true))

(fn lit_Mode
  (CallFunction|SetPP true)
  (CallFunction|SetupObjectList))

(fn MaterialAO
  (CallFunction|SetPP false false false false false false true))

(fn Metallic
  (CallFunction|SetPP false false false false true))

(fn MRQ_AlphaCutout
  (MRQ|MRQPostProcessSwitch "AlphaCutout"))

(fn MRQ_BaseColor
  (MRQ|MRQPostProcessSwitch "BaseColor"))

(fn MRQ_Emissive
  (MRQ|MRQPostProcessSwitch "Emissive"))

(fn MRQ_MaterialAO
  (MRQ|MRQPostProcessSwitch "MaterialAO"))

(fn MRQ_Metallic
  (MRQ|MRQPostProcessSwitch "Metallic"))

(fn MRQ_PostProcessSwitch
  (params (Pass :type Name))
  (CallFunction|LitMode)
  (switch name Pass
    (:BaseColor
      (CallFunction|UpdateWidgetPassText false true)
      (Variables|/Script/Engine.PostProcessComponent|SetbEnabled true (Variables|Default|GetPostProcess_BaseColor_MRQ))
      (Variables|/Script/Engine.PostProcessComponent|SetbEnabled false (Variables|Default|GetPostProcess_Metallic_MRQ))
      (Variables|/Script/Engine.PostProcessComponent|SetbEnabled false (Variables|Default|GetPostProcess_Roughness_MRQ))
      (Variables|/Script/Engine.PostProcessComponent|SetbEnabled false (Variables|Default|GetPostProcess_Specular_MRQ))
      (Variables|/Script/Engine.PostProcessComponent|SetbEnabled false (Variables|Default|GetPostProcess_MaterialAO_MRQ))
      (Variables|/Script/Engine.PostProcessComponent|SetbEnabled false (Variables|Default|GetPostProcess_SeperateTranslucency_MRQ))
      (Variables|/Script/Engine.PostProcessComponent|SetbEnabled false (Variables|Default|GetPostProcess_AlphaCutout)))
    (:Roughness
      (CallFunction|UpdateWidgetPassText false false true)
      (Variables|/Script/Engine.PostProcessComponent|SetbEnabled false (Variables|Default|GetPostProcess_BaseColor_MRQ))
      (Variables|/Script/Engine.PostProcessComponent|SetbEnabled false (Variables|Default|GetPostProcess_Metallic_MRQ))
      (Variables|/Script/Engine.PostProcessComponent|SetbEnabled true (Variables|Default|GetPostProcess_Roughness_MRQ))
      (Variables|/Script/Engine.PostProcessComponent|SetbEnabled false (Variables|Default|GetPostProcess_Specular_MRQ))
      (Variables|/Script/Engine.PostProcessComponent|SetbEnabled false (Variables|Default|GetPostProcess_MaterialAO_MRQ))
      (Variables|/Script/Engine.PostProcessComponent|SetbEnabled false (Variables|Default|GetPostProcess_SeperateTranslucency_MRQ))
      (Variables|/Script/Engine.PostProcessComponent|SetbEnabled false (Variables|Default|GetPostProcess_AlphaCutout)))
    (:Metallic
      (CallFunction|UpdateWidgetPassText false false false false true))
    (:Specular
      (CallFunction|UpdateWidgetPassText false false false true)
      (Variables|/Script/Engine.PostProcessComponent|SetbEnabled false (Variables|Default|GetPostProcess_BaseColor_MRQ))
      (Variables|/Script/Engine.PostProcessComponent|SetbEnabled false (Variables|Default|GetPostProcess_Metallic_MRQ))
      (Variables|/Script/Engine.PostProcessComponent|SetbEnabled false (Variables|Default|GetPostProcess_Roughness_MRQ))
      (Variables|/Script/Engine.PostProcessComponent|SetbEnabled true (Variables|Default|GetPostProcess_Specular_MRQ))
      (Variables|/Script/Engine.PostProcessComponent|SetbEnabled false (Variables|Default|GetPostProcess_MaterialAO_MRQ))
      (Variables|/Script/Engine.PostProcessComponent|SetbEnabled false (Variables|Default|GetPostProcess_SeperateTranslucency_MRQ))
      (Variables|/Script/Engine.PostProcessComponent|SetbEnabled false (Variables|Default|GetPostProcess_AlphaCutout)))
    (:Emissive
      (Variables|/Script/Engine.PostProcessComponent|SetbEnabled false (Variables|Default|GetPostProcess_BaseColor_MRQ))
      (Variables|/Script/Engine.PostProcessComponent|SetbEnabled false (Variables|Default|GetPostProcess_Metallic_MRQ))
      (Variables|/Script/Engine.PostProcessComponent|SetbEnabled false (Variables|Default|GetPostProcess_Roughness_MRQ))
      (Variables|/Script/Engine.PostProcessComponent|SetbEnabled false (Variables|Default|GetPostProcess_Specular_MRQ))
      (Variables|/Script/Engine.PostProcessComponent|SetbEnabled false (Variables|Default|GetPostProcess_MaterialAO_MRQ))
      (Variables|/Script/Engine.PostProcessComponent|SetbEnabled false (Variables|Default|GetPostProcess_SeperateTranslucency_MRQ))
      (CallFunction|Emissive)
      (Variables|/Script/Engine.PostProcessComponent|SetbEnabled false (Variables|Default|GetPostProcess_AlphaCutout)))
    (:MaterialAO
      (CallFunction|UpdateWidgetPassText false false false false false false true)
      (Variables|/Script/Engine.PostProcessComponent|SetbEnabled false (Variables|Default|GetPostProcess_BaseColor_MRQ))
      (Variables|/Script/Engine.PostProcessComponent|SetbEnabled false (Variables|Default|GetPostProcess_Metallic_MRQ))
      (Variables|/Script/Engine.PostProcessComponent|SetbEnabled false (Variables|Default|GetPostProcess_Roughness_MRQ))
      (Variables|/Script/Engine.PostProcessComponent|SetbEnabled false (Variables|Default|GetPostProcess_Specular_MRQ))
      (Variables|/Script/Engine.PostProcessComponent|SetbEnabled true (Variables|Default|GetPostProcess_MaterialAO_MRQ))
      (Variables|/Script/Engine.PostProcessComponent|SetbEnabled false (Variables|Default|GetPostProcess_SeperateTranslucency_MRQ))
      (Variables|/Script/Engine.PostProcessComponent|SetbEnabled false (Variables|Default|GetPostProcess_AlphaCutout)))
    (:SeperateTranslucency
      (CallFunction|UpdateWidgetPassText false false false false false false false true)
      (Variables|/Script/Engine.PostProcessComponent|SetbEnabled false (Variables|Default|GetPostProcess_BaseColor_MRQ))
      (Variables|/Script/Engine.PostProcessComponent|SetbEnabled false (Variables|Default|GetPostProcess_Metallic_MRQ))
      (Variables|/Script/Engine.PostProcessComponent|SetbEnabled false (Variables|Default|GetPostProcess_Roughness_MRQ))
      (Variables|/Script/Engine.PostProcessComponent|SetbEnabled false (Variables|Default|GetPostProcess_Specular_MRQ))
      (Variables|/Script/Engine.PostProcessComponent|SetbEnabled false (Variables|Default|GetPostProcess_MaterialAO_MRQ))
      (Variables|/Script/Engine.PostProcessComponent|SetbEnabled true (Variables|Default|GetPostProcess_SeperateTranslucency_MRQ))
      (Variables|/Script/Engine.PostProcessComponent|SetbEnabled false (Variables|Default|GetPostProcess_AlphaCutout)))
    (:SSCScheen
      (CallFunction|UpdateWidgetPassText false false false false false false false false true))
    (:SSCAOThreshold
      (CallFunction|UpdateWidgetPassText false false false false false false false false false true))
    (:AlphaCutout
      (Variables|/Script/Engine.PostProcessComponent|SetbEnabled false (Variables|Default|GetPostProcess_BaseColor_MRQ))
      (Variables|/Script/Engine.PostProcessComponent|SetbEnabled false (Variables|Default|GetPostProcess_Metallic_MRQ))
      (Variables|/Script/Engine.PostProcessComponent|SetbEnabled false (Variables|Default|GetPostProcess_Roughness_MRQ))
      (Variables|/Script/Engine.PostProcessComponent|SetbEnabled false (Variables|Default|GetPostProcess_Specular_MRQ))
      (Variables|/Script/Engine.PostProcessComponent|SetbEnabled false (Variables|Default|GetPostProcess_MaterialAO_MRQ))
      (Variables|/Script/Engine.PostProcessComponent|SetbEnabled false (Variables|Default|GetPostProcess_SeperateTranslucency_MRQ))
      (Variables|/Script/Engine.PostProcessComponent|SetbEnabled true (Variables|Default|GetPostProcess_AlphaCutout))))
  (Variables|/Script/Engine.PostProcessComponent|SetbEnabled false (Variables|Default|GetPostProcess_BaseColor_MRQ))
  (Variables|/Script/Engine.PostProcessComponent|SetbEnabled true (Variables|Default|GetPostProcess_Metallic_MRQ))
  (Variables|/Script/Engine.PostProcessComponent|SetbEnabled false (Variables|Default|GetPostProcess_Roughness_MRQ))
  (Variables|/Script/Engine.PostProcessComponent|SetbEnabled false (Variables|Default|GetPostProcess_Specular_MRQ))
  (Variables|/Script/Engine.PostProcessComponent|SetbEnabled false (Variables|Default|GetPostProcess_MaterialAO_MRQ))
  (Variables|/Script/Engine.PostProcessComponent|SetbEnabled false (Variables|Default|GetPostProcess_SeperateTranslucency_MRQ))
  (Variables|/Script/Engine.PostProcessComponent|SetbEnabled false (Variables|Default|GetPostProcess_AlphaCutout)))

(fn MRQ_Roughness
  (MRQ|MRQPostProcessSwitch "Roughness"))

(fn MRQ_SeperateTranslucency
  (MRQ|MRQPostProcessSwitch "SeperateTranslucency"))

(fn MRQ_Specular
  (MRQ|MRQPostProcessSwitch "Specular"))

(fn MRQ_SSCAOThreshold
  (MRQ|MRQPostProcessSwitch "SSCAOThreshold"))

(fn MRQ_SSCSheen
  (MRQ|MRQPostProcessSwitch "SSCScheen"))

(fn roughness
  (CallFunction|SetPP false false true))

(fn SeperateTranslucencyAlpha
  (CallFunction|SetPP false false false false false false false true))

(fn setPP
  (params (Lit :type Boolean) (BaseColor :type Boolean) (Roughness :type Boolean) (Specular :type Boolean) (Metallic :type Boolean) (Emissive :type Boolean) (MaterialAO :type Boolean) (Seperate%20Translucency :type Boolean) (SSC%20Sheen :type Boolean) (SSC%20AO%20Threshold :type Boolean) (World%20Normal :type Boolean))
  (Utilities|FlowControl|Sequence
    (:then_0
      ;[1 1 1] PostProcess Enable
      (Variables|/Script/Engine.PostProcessComponent|SetbEnabled BaseColor (Variables|Default|GetPostProcess_BaseColor_Viewport))
      (Variables|/Script/Engine.PostProcessComponent|SetbEnabled Roughness (Variables|Default|GetPostProcess_Roughness_Viewport))
      (Variables|/Script/Engine.PostProcessComponent|SetbEnabled (or SSC%20Sheen Specular) (Variables|Default|GetPostProcess_Specular_Viewport))
      (Variables|/Script/Engine.PostProcessComponent|SetbEnabled (or SSC%20AO%20Threshold Metallic) (Variables|Default|GetPostProcess_Metallic_Viewport))
      (Variables|/Script/Engine.PostProcessComponent|SetbEnabled MaterialAO (Variables|Default|GetPostProcess_MaterialAO_MRQ))
      (Variables|/Script/Engine.PostProcessComponent|SetbEnabled Seperate%20Translucency (Variables|Default|GetPostProcess_SeperateTranslucency_MRQ))
      (Variables|/Script/Engine.PostProcessComponent|SetbEnabled World%20Normal (Variables|Default|GetPostProcess_Normals)))
    (:then_1
      (CallFunction|UpdateWidgetPassText Lit BaseColor Roughness Specular Metallic Emissive MaterialAO Seperate%20Translucency SSC%20Sheen SSC%20AO%20Threshold World%20Normal))
    (:then_2
      ;[1 1 1] AA settings
      (if Lit
        (Development|ExecuteConsoleCommand "r.TemporalAASamples 8")
        (else
          (Development|ExecuteConsoleCommand "r.TemporalAASamples 1"))))
    (:then_3
      ;[1 1 1] Emissive
      (if Emissive
        (Development|ExecuteConsoleCommand "ShowFlag.Diffuse 0")
        (Development|ExecuteConsoleCommand "ShowFlag.Specular 0")
        (Development|ExecuteConsoleCommand "ShowFlag.Lighting 0")
        (else
          (Development|ExecuteConsoleCommand "ShowFlag.Diffuse 1")
          (Development|ExecuteConsoleCommand "ShowFlag.Specular 1")
          (Development|ExecuteConsoleCommand "ShowFlag.Lighting 1"))))))

(fn setup
  (CallFunction|SetupShadows)
  (CallFunction|SetupObjectList)
  (CallFunction|LitMode)
  (Class|WBPTurntableUI|SetText (Variables|Default|GetUI) (Variables|UIValues|GetuserName) (Variables|UIValues|Getdate) (Variables|UIValues|Getbuild) (Variables|UIValues|GetobjectList) (Variables|UIValues|Getversion) "" "")
  (Development|ExecuteConsoleCommand "r.MaterialQualityLevel 1")
  (CallFunction|LitMode))

(fn setupLevelsChecker
  (if (not (Utilities|IsValid (Variables|Default|GetdynamicMaterialPP_levelsChecker) :!class /Script/Engine.KismetSystemLibrary))
    (bind _returnvalue (Rendering|Material|CreateDynamicMaterialInstance "/Gbuffer_Turntable/FN_PublicTurntable_Assets/LevelsChecker/MAT_LevelsChecker_BaseColor.MAT_LevelsChecker_BaseColor" "None" "None" :!class /Script/Engine.KismetMaterialLibrary))
    (bind _output_get (Variables|Default|SetdynamicMaterialPP_levelsChecker _returnvalue))
    (bind _returnvalue_2 (Rendering|Material|CreateDynamicMaterialInstance "/Gbuffer_Turntable/FN_PublicTurntable_Assets/LevelsChecker/MAT_LevelsChecker_BaseColor.MAT_LevelsChecker_BaseColor" "None" "None" :!class /Script/Engine.KismetMaterialLibrary))
    (bind _output_get_2 (Variables|Default|SetdynamicMaterialPP_baseColor _returnvalue_2)))
  (Rendering|Material|SetScalarParameterValue (Variables|Default|GetdynamicMaterialPP_levelsChecker) "MinValue" (Variables|LevelsChecker|GetminValue) :!class /Script/Engine.MaterialInstanceDynamic)
  (Rendering|Material|SetScalarParameterValue (Variables|Default|GetdynamicMaterialPP_levelsChecker) "MaxValue" (Variables|LevelsChecker|GetmaxValue) :!class /Script/Engine.MaterialInstanceDynamic)
  (Rendering|Material|SetVectorParameterValue (Variables|Default|GetdynamicMaterialPP_levelsChecker) "MinValueColor" (Variables|LevelsChecker|GetminValueColor) :!class /Script/Engine.MaterialInstanceDynamic)
  (Rendering|Material|SetVectorParameterValue (Variables|Default|GetdynamicMaterialPP_levelsChecker) "MaxValueColor" (Variables|LevelsChecker|GetmaxValueColor) :!class /Script/Engine.MaterialInstanceDynamic)
  (Utilities|Struct|SetmembersinPostProcessSettings (Variables|/Script/Engine.PostProcessComponent|GetSettings (Variables|Default|GetPostProcess_BaseColor_Viewport)) (Utilities|Struct|MakeWeightedBlendables (Utilities|Array|MakeArray (Utilities|Struct|MakeWeightedBlendable (Math|Conversions|ToFloat(Boolean) (Variables|LevelsChecker|Getenabled)) _output_get) (Utilities|Struct|MakeWeightedBlendable 1.0 _output_get_2)))))

(fn setupObjectList
  (bind _objectlist (Variables|UIValues|GetobjectList))
  (Development|PrintString "Setup Object List" false)
  (Variables|UIValues|SetobjectList "")
  (for _array_element (Components|GetChildrenComponents (Variables|/Script/Engine.Actor|GetRootComponent (Variables|Default|GetTurntable_Objects)) false)
    (bind _asstatic_mesh_component (Utilities|Casting|CastToStaticMeshComponent _array_element)
      (:then
        (bind _returnvalue (CallFunction|VertexCount (Variables|/Script/Engine.StaticMeshComponent|GetStaticMesh _asstatic_mesh_component)))
        (Variables|UIValues|SetobjectList (Utilities|String|Append _objectlist (Utilities|GetDisplayName (Variables|/Script/Engine.StaticMeshComponent|GetStaticMesh _asstatic_mesh_component)) " (" (Utilities|String|ToString(Integer) _returnvalue) ")
")))
      (:CastFailed
        (bind _asskeletal_mesh_component (Utilities|Casting|CastToSkeletalMeshComponent _array_element)
          (:then
            (bind _returnvalue_2 (CallFunction|VertexCountSkeletalMesh (Variables|/Script/Engine.SkinnedMeshComponent|GetSkeletalMesh _asskeletal_mesh_component)))
            (Variables|UIValues|SetobjectList (Utilities|String|Append _objectlist (Utilities|GetDisplayName (Variables|/Script/Engine.SkinnedMeshComponent|GetSkeletalMesh _asskeletal_mesh_component)) " (" (Utilities|String|ToString(Integer) _returnvalue_2) ")
")))
          (:CastFailed
            (Development|PrintString "Error casting to component in SetupObjectList")))))))

(fn setupShadows
  (Utilities|FlowControl|Sequence
    (:then_0
      (Development|ExecuteConsoleCommand "r.Shadow.PerObjectDirectionalDepthBias 10"))
    (:then_1
      (Development|ExecuteConsoleCommand "r.Shadow.PerObjectDirectionalSlopeDepthBias 10"))
    (:then_2
      (for _array_element (Components|GetChildrenComponents (Variables|/Script/Engine.Actor|GetRootComponent (Variables|Default|GetTurntable_Objects)) false)
        (bind _asprimitive_component (Utilities|Casting|CastToPrimitiveComponent _array_element))
        (Rendering|SetCastInsetShadow _asprimitive_component true)))))

(fn specular
  (CallFunction|SetPP false false false true))

(fn SSCAOThreshold
  (CallFunction|SetPP false false false false false false false false false true))

(fn SSCScheen
  (CallFunction|SetPP false false false false false false false false true))

(fn ToggleUI
  (params (Visiblity? :type Boolean))
  (bind _ui (Variables|Default|GetUI))
  (if Visiblity?
    (Widget|SetVisibility _ui "Visible")
    (else
      (Widget|SetVisibility _ui "Hidden"))))

(fn UpdateWidgetPassText
  (params (IsStudio :type Boolean) (IsBaseColor :type Boolean) (IsRoughness :type Boolean) (IsSpecular :type Boolean) (IsMetallic :type Boolean) (IsEmissive :type Boolean) (IsMaterialAO :type Boolean) (IsSeperateTranslucency :type Boolean) (IsSSCSheen :type Boolean) (IsSSCAOThreshold :type Boolean) (World%20Normal :type Boolean))
  (bind _ui (Variables|Default|GetUI))
  (bind _aswbp_turntable_ui (Utilities|Casting|CastToWBP_Turntable_UI _ui))
  (bind _ui_2 (Variables|Default|GetUI))
  (bind _checkercontrolbox (Variables|/Gbuffer_Turntable/FN_PublicTurntable_Assets/BP/WBP_Turntable_UI.WBP_Turntable_UI_C|GetCheckerControlBox _aswbp_turntable_ui))
  (Utilities|FlowControl|Sequence
    (:then_0
      ;[1 1 1] Checker Legend Visibility
      (Widget|SetVisibility _checkercontrolbox "Hidden"))
    (:then_1
      ;[1 1 1] UI Pass
      (if IsStudio
        (Class|WBPTurntableUI|SetPass _ui_2 "Studio Lighting")
        (elif IsBaseColor
          (Class|WBPTurntableUI|SetPass _ui_2 "BaseColor")
          (Widget|SetVisibility _checkercontrolbox "Visible")
          (elif IsRoughness
            (Class|WBPTurntableUI|SetPass _ui_2 "Roughness")
            (elif IsSpecular
              (Class|WBPTurntableUI|SetPass _ui_2 "Specular")
              (elif IsMetallic
                (Class|WBPTurntableUI|SetPass _ui_2 "Metallic")
                (elif IsEmissive
                  (Class|WBPTurntableUI|SetPass _ui_2 "Emissive")
                  (elif IsMaterialAO
                    (Class|WBPTurntableUI|SetPass _ui_2 "Material AO")
                    (elif IsSeperateTranslucency
                      (Class|WBPTurntableUI|SetPass _ui_2 "Seperate Translucency")
                      (elif IsSSCSheen
                        (Class|WBPTurntableUI|SetPass _ui_2 "SSC Sheen")
                        (elif IsSSCAOThreshold
                          (Class|WBPTurntableUI|SetPass _ui_2 "SSC AO Threshold")
                          (elif World%20Normal
                            (Class|WBPTurntableUI|SetPass _ui_2 "SSC World Normal")
                            (else
                              (Class|WBPTurntableUI|SetPass _ui_2 "Unknown"))))))))))))))))

(fn UpdatePassTestCustom
  (params (Pass :type String))
  (Class|WBPTurntableUI|SetPass (Variables|Default|GetUI) Pass))

(fn UpdateTODM
  (params (InTimeOfDayManager :type Info%20Soft%20Class%20Reference) (ShowTODM? :type Boolean) (New%20Rotation%20Z%20%28Yaw%29 :type Float%20%28single-precision%29))
  (bind _returnvalue (TimeOfDay|GetGlobalTimeOfDayManager))
  (bind _returnvalue_2 (Fortnite|GetFortWorldSettingsinEditor))
  (TimeOfDay|SetGlobalTimeOfDayManager InTimeOfDayManager)
  (if (Utilities|IsValidInterface _returnvalue)
    (Rendering|Components|Light|SetLightFunctionFadeDistance (TimeOfDay|GetDirectionalLightComponent _returnvalue) 0.0)
    (Development|PrintString "----------------------Attempting Visibility Change of TODM")
    (Development|PrintString (Utilities|String|ToString(Boolean) ShowTODM?))
    (if ShowTODM?
      (TimeOfDay|DisableAllTimeOfDayManagers _returnvalue_2 false)
      (TimeOfDay|SetTimeOfDayManagerVisibility _returnvalue true)
      (bind _outactors (Actor|GetAllActorsOfClass "/Script/FortniteGame.FortDaySequenceActor"))
      (if (> (Utilities|Array|Length _outactors) 0)
        (Transformation|SetActorRotation (Utilities|Array|Get(acopy) _outactors) (Math|Rotator|MakeRotator 0.0 0.0 New%20Rotation%20Z%20%28Yaw%29) false)
        (else
          (Development|PrintString "~~~~~~~~~~~~~~~~~~~~~~~~~~~~~~~~~~~~~~~~~~~CANNOT FIND TODM FOR ROTATION~~~~~~~~~~~~~~~~~~~~~~~~~")))
      (else
        (TimeOfDay|DisableAllTimeOfDayManagers _returnvalue_2 true)
        (TimeOfDay|SetTimeOfDayManagerVisibility _returnvalue false)))
    (else
      (Development|PrintString (Utilities|String|Append "----------------------Global TODM is invalid (class was \"" (Utilities|ToString(SoftClassReference) InTimeOfDayManager) "\")")))))

(fn ConstructionScript)

(fn vertexCount
  (params (staticMesh :type Static%20Mesh%20Object%20Reference))
  (bind _returnvalue (Utilities|Name|MakeLiteralName "Vertices"))
  (bind _returnvalue_2 (AssetData|CreateAssetData staticMesh))
  (bind (_outtagvalue _returnvalue_3) (AssetData|GetTagValue _returnvalue_2 _returnvalue))
  (if _returnvalue_3
    (return (Utilities|String|StringToInteger _outtagvalue))
    (else
      (return -1))))

(fn vertexCount_SkeletalMesh
  (params (skeletalMEsh :type Skeletal%20Mesh%20Object%20Reference))
  (bind _returnvalue (Utilities|Name|MakeLiteralName "Vertices"))
  (bind _returnvalue_2 (AssetData|CreateAssetData skeletalMEsh))
  (bind (_outtagvalue _returnvalue_3) (AssetData|GetTagValue _returnvalue_2 _returnvalue))
  (if _returnvalue_3
    (return (Utilities|String|StringToInteger _outtagvalue))
    (else
      (return -1))))

(fn WorldNormal
  (CallFunction|SetPP true false false false false false false false false false true))
