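; Tech-art turntable Blueprint that previews and batch-renders item wraps on weapon and vehicle skeletal meshes.
; Construction and next/previous functions apply the selected wrap to target actors, then center and scale them using per-object data-table values and optional weapon auto-scaling.
; Python console commands set vehicle material-section masks and launch batch rendering.
; At play, an optional widget overlay shows user, date, build and pass, with capture frame number updated on Tick.
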(event EventBeginPlay
  (bind _returnvalue (UserInterface|CreateWidget "/Game/NeverCook/TechArt/WrapTurntable/WBP_WrapTurntable_UI.WBP_WrapTurntable_UI_C"))
  (bind _output_get (Variables|Default|SetUI _returnvalue))
  (if (Variables|RenderSettings|GetshowUI)
    (UserInterface|Viewport|AddtoViewport _output_get 0)
    (Class|WBPWrapTurntableUI|SetText _output_get (Variables|Default|Getuser) (Variables|Default|Getdate) (Variables|Default|Getbuild) (Variables|Targets|Getpass) "")))

(event Collision|EventActorBeginOverlap
  (params (OtherActor :type Actor%20Object%20Reference)))

(event EventTick
  (params (DeltaSeconds :type Float%20%28single-precision%29))
  (Class|WBPWrapTurntableUI|SetFrameNumber (Variables|Default|GetUI) (Cinematics|Capture|GetCaptureFrameNumber)))

(event EventEndPlay
  (params (EndPlayReason :type EEndPlayReason%20Enum :default "Destroyed")))

(fn ApplyWrap
  (params (Wrap :type Athena%20Item%20Wrap%20Definition%20Object%20Reference) (Mesh :type Skeletal%20Mesh%20Object%20Reference) (Target :type Fort%20Wrap%20Skeletal%20Mesh%20Actor%20Object%20Reference) (vehicle :type Boolean) (maskingValue :type Integer))
  (Components|SkinnedMesh|SetSkinnedAssetandUpdate (Variables|/Script/Engine.SkeletalMeshActor|GetSkeletalMeshComponent Target) Mesh true)
  (if vehicle
    (Class|FortWrapSkeletalMeshActor|SetWrap Target Wrap "VehicleWrap_Opaque")
    (Development|ExecuteConsoleCommand (Utilities|String|Append "py Techart/WrapTurntable/setMaterialSectionMask.py" " " (Utilities|GetDisplayName Target) " " (Utilities|String|ToString(Integer) maskingValue)))
    (else
      (Cosmetic|ApplyItemWraptoActor Wrap Target "VehicleWrap_Opaque"))))

(fn BatchRender
  (Development|ExecuteConsoleCommand (Utilities|String|Append "py Techart/WrapTurntable/wrapTurntable.py " (Utilities|GetObjectName self))))

(fn centerMesh
  (params (Actor :type Actor%20Object%20Reference) (scale :type Float%20%28double-precision%29))
  (Transformation|SetActorRelativeLocation Actor "0, 0, 0" false true)
  (Transformation|SetActorScale3D Actor (Math|Vector|MakeVector scale scale scale))
  (Transformation|AddActorWorldOffset Actor (- (Transformation|GetWorldLocation (Variables|/Script/Engine.Actor|GetRootComponent Actor)) (Collision|GetActorBounds Actor false false)) false true))

(fn checkSize
  (Default|PrintMeshBounds (Variables|Settings|GetWeapon)))

(fn compileMaterials
  (for _array_element (Variables|Settings|GetWraps)
    (Variables|Settings|SetWrap _array_element)
    (CallFunction|UpdateMeshandWrap)))

(fn DrawBounds
  (params (Target :type Actor%20Object%20Reference) (Thickness :type Float%20%28double-precision%29))
  (bind (_origin _boxextent) (Collision|GetActorBounds Target false false))
  (Rendering|Debug|DrawDebugBox :Center _origin :Extent _boxextent :LineColor "(R=0.135193,G=0.519008,B=1.000000,A=1.000000)" :Duration 5.0 :Thickness Thickness :DepthPriority "World")
  (Rendering|Debug|DrawDebugBox :Center _origin :Extent "0.100000,0.100000,0.100000" :LineColor "(R=1.000000,G=0.031132,B=0.110808,A=1.000000)" :Duration 5.0 :Thickness 1.0 :DepthPriority "World")
  (Rendering|Debug|DrawDebugBox :Center (Transformation|GetActorLocation Target) :Extent "0.100000,0.100000,0.100000" :LineColor "(R=1.000000,G=0.906202,B=0.000000,A=1.000000)" :Duration 5.0 :Thickness 1.0 :DepthPriority "World"))

(fn GetActorSize
  (params (Actor :type Actor%20Object%20Reference))
  (bind (_origin _boxextent) (Collision|GetActorBounds Actor false false))
  (return (* (Math|Vector|VectorLength (- _boxextent _origin)) 2.0)))

(fn GetDataTableValues
  (bind (_vehicle _rotate90 _mask _scale) (CallFunction|GetObjectData (Variables|Settings|GetVehicle)))
  (Variables|Settings|SetRotateVehicle90 _rotate90)
  (Variables|Settings|SetvehicleMaskingValue _mask)
  (Variables|Settings|SetvehicleScale _scale)
  (bind (_vehicle_2 _rotate90_2 _mask_2 _scale_2) (CallFunction|GetObjectData (Variables|Settings|GetWeapon)))
  (Variables|Settings|SetweaponMaskingValue _mask_2)
  (Variables|Settings|SetweaponScale _scale_2))

(fn GetMeshSize
  (params (SkeletalMesh :type Skeletal%20Mesh%20Object%20Reference))
  (bind (_returnvalue_origin _returnvalue_boxextent _returnvalue_sphereradius) (Math|BoxSphereBounds|BreakBoxSphereBounds (Mesh|GetBounds SkeletalMesh)))
  (return _returnvalue_sphereradius))

(fn getObjectData
  (params (Object :type Object%20Reference))
  (bind _returnvalue (Utilities|GetDataTableRow (Variables|Settings|GetwrapObjectDataTable) (Utilities|String|StringToName (Utilities|GetDisplayName Object)))
    (:then
      (return (.Vehicle_10_5431444541D6EBA6C5034CB397B35C14 (Utilities|Struct|BreakWrapTurntableObjectStructure _returnvalue)) (.rotate90_12_9EAFDEF64E1912B33BAED18E8ADC1E63 (Utilities|Struct|BreakWrapTurntableObjectStructure _returnvalue)) (.Mask_11_4A1952414E86F7779E81B59C79B935E6 (Utilities|Struct|BreakWrapTurntableObjectStructure _returnvalue)) (.Scale_15_6C7B332244F53AB662F2F2980F7D4B75 (Utilities|Struct|BreakWrapTurntableObjectStructure _returnvalue))))
    (:RowNotFound
      (return false false 1 1.0))))

(fn nextVehicle
  (for (_array_element _array_index) (Variables|Settings|GetVehicles)
    (if (Utilities|Equal(Object) _array_element (Variables|Settings|GetVehicle))
      (Variables|Default|Seti _array_index)))
  (bind _result (Math|Integer|IncrementInt (Variables|Default|Geti))
    (:%20%20
      (if (>= _result (Utilities|Array|Length (Variables|Settings|GetVehicles)))
        (Variables|Default|Seti 0))
      (Variables|Settings|SetVehicle (Utilities|Array|Get(acopy) (Variables|Settings|GetVehicles) (Variables|Default|Geti)))
      (CallFunction|GetDataTableValues)
      (CallFunction|UpdateMeshandWrap)
      (CallFunction|UpdateCamerasandMeshTransforms)
      (CallFunction|SnapCamera (Variables|Targets|GetCamera) (Utilities|Array|Get(acopy) (Variables|Targets|GetCameras) 2)))))

(fn nextWeapon
  (for (_array_element _array_index) (Variables|Settings|GetWeapons)
    (if (Utilities|Equal(Object) _array_element (Variables|Settings|GetWeapon))
      (Variables|Default|Seti _array_index)))
  (bind _result (Math|Integer|IncrementInt (Variables|Default|Geti))
    (:%20%20
      (if (>= _result (Utilities|Array|Length (Variables|Settings|GetWeapons)))
        (Variables|Default|Seti 0))
      (Variables|Settings|SetWeapon (Utilities|Array|Get(acopy) (Variables|Settings|GetWeapons) (Variables|Default|Geti)))
      (CallFunction|GetDataTableValues)
      (CallFunction|UpdateMeshandWrap)
      (CallFunction|UpdateCamerasandMeshTransforms)
      (CallFunction|SnapCamera (Variables|Targets|GetCamera) (Utilities|Array|Get(acopy) (Variables|Targets|GetCameras))))))

(fn nextWrap
  (for (_array_element _array_index) (Variables|Settings|GetWraps)
    (if (Utilities|Equal(Object) _array_element (Variables|Settings|GetWrap))
      (Variables|Default|Seti _array_index)))
  (bind _result (Math|Integer|IncrementInt (Variables|Default|Geti))
    (:%20%20
      (if (>= _result (Utilities|Array|Length (Variables|Settings|GetWraps)))
        (Variables|Default|Seti 0))
      (Variables|Settings|SetWrap (Utilities|Array|Get(acopy) (Variables|Settings|GetWraps) (Variables|Default|Geti)))
      (CallFunction|UpdateMeshandWrap)
      (CallFunction|UpdateCamerasandMeshTransforms))))

(fn populateWrapList
  (Utilities|Array|Clear (Variables|Settings|GetWraps))
  (bind _returnvalue (Fortnite|JonLHackGetAllObjectsOfClassfromPath "/Game/Athena/Items/Cosmetics/ItemWraps" "/Script/FortniteGame.AthenaItemWrapDefinition"))
  (for _array_element _returnvalue
    (bind _asathena_item_wrap_definition (Utilities|Casting|CastToAthenaItemWrapDefinition _array_element))
    (if (Utilities|String|Contains (Utilities|GetDisplayName _asathena_item_wrap_definition) "Wrap_")
      (if (not (Utilities|String|Contains (Utilities|GetDisplayName _asathena_item_wrap_definition) "_TBD_"))
        (Utilities|Array|Add (Variables|Settings|GetWraps) _asathena_item_wrap_definition)))))

(fn previousWrap
  (bind _wraps (Variables|Settings|GetWraps))
  (for (_array_element _array_index) (Variables|Settings|GetWraps)
    (if (Utilities|Equal(Object) _array_element (Variables|Settings|GetWrap))
      (Variables|Default|Seti _array_index)))
  (bind _result (Math|Integer|DecrementInt (Variables|Default|Geti))
    (:%20%20
      (if (< _result 0)
        (Variables|Default|Seti (- (Utilities|Array|Length _wraps) 1)))
      (Variables|Settings|SetWrap (Utilities|Array|Get(acopy) _wraps (Variables|Default|Geti)))
      (CallFunction|UpdateMeshandWrap)
      (CallFunction|UpdateCamerasandMeshTransforms))))

(fn printMeshBounds
  (params (SkeletalMesh :type Skeletal%20Mesh%20Object%20Reference))
  (bind (_returnvalue_origin _returnvalue_boxextent) (Math|BoxSphereBounds|BreakBoxSphereBounds (Mesh|GetBounds SkeletalMesh)))
  (Development|PrintString (Utilities|String|ToString(Vector) _returnvalue_origin))
  (Development|PrintString (Utilities|String|ToString(Vector) _returnvalue_boxextent)))

(fn setupObjectList
  (Variables|Targets|Setpass (Utilities|GetDisplayName (Variables|Settings|GetWrap)))
  (Variables|Targets|SetobjectList (Utilities|String|Append (Utilities|GetDisplayName (Variables|Settings|GetWeapon)) ", " (Utilities|GetDisplayName (Variables|Settings|GetVehicle)))))

(fn snapCamera
  (params (target :type Camera%20Actor%20Object%20Reference) (source :type Camera%20Actor%20Object%20Reference))
  (bind _cameracomponent (Variables|/Script/Engine.CameraActor|GetCameraComponent source))
  (bind (_returnvalue_location _returnvalue_rotation) (Math|Transform|BreakTransform (Transformation|GetWorldTransform _cameracomponent)))
  (bind _cameracomponent_2 (Variables|/Script/Engine.CameraActor|GetCameraComponent target))
  (Transformation|SetWorldLocationAndRotation _cameracomponent_2 _returnvalue_location _returnvalue_rotation false true)
  (Camera|SetFieldOfView _cameracomponent_2 (Variables|/Script/Engine.CameraComponent|GetFieldOfView _cameracomponent)))

(fn ToggleBounds
  (for _array_element (Variables|Targets|GetWeaponTargets)
    (CallFunction|DrawBounds _array_element 0.5))
  (for _array_element_2 (Variables|Targets|GetVehicleTargets)
    (CallFunction|DrawBounds _array_element_2 1.0)))

(fn UpdateCamerasAndMeshTransforms
  (if (> (Utilities|Array|Length (Variables|Targets|GetCameras)) 0)
    (Utilities|FlowControl|Sequence
      (:then_0
        ;[1 1 1] get weapon and vehicles sizes
        (bind _size (Default|GetMeshSize (Variables|Settings|GetWeapon)))
        (Variables|Settings|SetweaponSize _size))
      (:then_1)
      (:then_2
        ;[1 1 1] WeaponB
        (for _array_element (Variables|Targets|GetWeaponTargets)
          (if (Variables|Settings|GetautoScale)
            (CallFunction|CenterMesh _array_element (* (Variables|Settings|GetweaponScale) (+ (/ 2e+01 (Variables|Settings|GetweaponSize)) 0.0)))
            (else
              (CallFunction|CenterMesh 0 (Variables|Settings|GetweaponScale))))))
      (:then_3
        ;[1 1 1] append
        (CallFunction|GetDataTableValues)
        (for _array_element_2 (Variables|Targets|GetVehicleTargets)
          (Transformation|SetActorRelativeRotation _array_element_2 (Math|Rotator|MakeRotator 0.0 0.0 (* (Math|Conversions|ToInteger(Boolean) (Variables|Settings|GetRotateVehicle90)) -9e+01)) false true)
          (CallFunction|CenterMesh _array_element_2 (Variables|Settings|GetvehicleScale)))))))

(fn UpdateMeshAndWrap
  (Utilities|FlowControl|Sequence
    (:then_0
      (for _array_element (Variables|Targets|GetWeaponTargets)
        (CallFunction|ApplyWrap (Variables|Settings|GetWrap) (Variables|Settings|GetWeapon) _array_element false (Variables|Settings|GetweaponMaskingValue))))
    (:then_1
      (for _array_element_2 (Variables|Targets|GetVehicleTargets)
        (CallFunction|ApplyWrap (Variables|Settings|GetWrap) (Variables|Settings|GetVehicle) _array_element_2 true (Variables|Settings|GetvehicleMaskingValue))))))

(fn ConstructionScript
  ;[1 1 1] UserName
  (Variables|Default|Setuser (Utilities|Platform|GetPlatformUserName))
  ;[1 1 1] Date
  (Variables|Default|Setdate (Utilities|String|ToString(Text) (Utilities|Text|AsDate(fromUTC) (Math|DateTime|UTCNow) "EDT")))
  ;[1 1 1] Build #
  (Variables|Default|Setbuild (Development|GetEngineVersion))
  ;[1 1 1] Object List
  (CallFunction|UpdateCamerasandMeshTransforms)
  (CallFunction|UpdateMeshandWrap))
