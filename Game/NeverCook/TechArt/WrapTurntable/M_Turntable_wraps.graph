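# M_Turntable_wraps is a Default Lit material in the NeverCook tech-art WrapTurntable folder, apparently a plastic surface for a turntable used to present wraps.
# A NormalTiling parameter scales the UVs of a rough-plastic normal map, which FlattenNormal then softens with a constant flatness.
# Scalar parameters, including Value and Roughness, supply tunable shading constants.
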Material /Game/NeverCook/TechArt/WrapTurntable/M_Turntable_wraps
graph {
 properties {pixelDepthOffsetMode=PDOM_Legacy, usedShadingModels=Default Lit}
 nodes [
  n1 = TextureCoordinate()
  n2 = ScalarParameter() {defaultValue=30, parameterName=NormalTiling}
  MaterialFunctionCall(
   Normal=TextureSample(UVs=Multiply(A=@n1, B=@n2)) {samplerType=SAMPLERTYPE_Normal, texture=/Game/FN_PublicTurntable/TurntableAssets/T_Rough_Plastic_N}.RGB
   Flatness=Constant() {r=0.8}
  ) {materialFunction=&FlattenNormal}
  ScalarParameter() {defaultValue=0.14, parameterName=Value}
  ScalarParameter() {defaultValue=0.25}
  ScalarParameter() {defaultValue=0.8, parameterName=Roughness}
 ]
}
