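# Movie Render Pipeline burn-in widget for the wrap turntable, overlaying name, date, build number, pass and frame number text on rendered output.
# OnOutputFrameStarted reads the pipeline's root frame number and passes it to setFrameNumber on each output frame.
# setText and setPass functions fill the other labels.
# An unwired setText call in the event graph references BP Turntable GBuffers fields, and setPass is not called here.
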WidgetBlueprint /Game/NeverCook/TechArt/WrapTurntable/WBP_WrapTurntable_UI
graphs [
 EventGraph {nodes [
   n1 = Event OnOutputFrameStarted!() -> (OutputDelegate, then, ForPipeline:Movie Pipeline Object Reference) {is=title K2Node_Event}
   > setFrameNumber!(
    execute
    self
    frameNumber:int=Break Frame Number!(FrameNumber:"Frame Number Structure (by ref)"=GetRootFrameNumber!(InMoviePipeline:Movie Pipeline Object Reference=@n1.ForPipeline) -> (ReturnValue:Frame Number Structure)) -> (Value:int) {is=title K2Node_BreakStruct}
   ) -> (then)
   setText!(
    execute
    self=Self-Reference()
    Name:String=Get userName!(self:BP Turntable GBuffers Object Reference) -> (userName:String)
    Date:String=Get date!(self:BP Turntable GBuffers Object Reference) -> (date:String)
    build:String=Get build!(self:BP Turntable GBuffers Object Reference) -> (build:String)
    Pass:String=lit
    FrameNumber:String=0001
    objectList:String=Get objectList!(self:BP Turntable GBuffers Object Reference) -> (objectList:String)
    Version:String=Get version!(self:BP Turntable GBuffers Object Reference) -> (version:String)
   ) -> (then)
   Event PreConstruct()
   Event Construct()
   Event Tick()
  ]}
 setFrameNumber {nodes [
   n1 = setFrameNumber!() -> (then, frameNumber:int) {is=title K2Node_FunctionEntry}
   > "SetText (Text)"(self=Get frameNumber() -> (frameNumber:Text Object Reference), InText="To Text (Integer)"(Value=@n1))
  ]}
 setPass {nodes [
   n1 = setPass!() -> (then, Pass:String) {is=title K2Node_FunctionEntry}
   > "SetText (Text)"(self=Get Pass() -> (Pass:Text Object Reference), InText="To Text (String)"(InString=@n1))
  ]}
 setText {nodes [
   n1 = setText!() -> (then, Name:String, Date:String, build:String, Pass:String, FrameNumber:String) {is=title K2Node_FunctionEntry}
   > "SetText (Text)"(self=Get Name() -> (Name:Text Object Reference), InText="To Text (String)"(InString=@n1))
   > "SetText (Text)"(self=Get Date() -> (Date:Text Object Reference), InText="To Text (String)"(InString=@n1.Date))
   > "SetText (Text)"(self=Get BuildNumber() -> (BuildNumber:Text Object Reference), InText="To Text (String)"(InString=@n1.build))
   > "SetText (Text)"(self=Get Pass(), InText="To Text (String)"(InString=@n1.Pass))
   > "SetText (Text)"(self=Get frameNumber(), InText="To Text (String)"(InString=@n1.FrameNumber))
  ]}
]
variables []
widgets {
 info {
  inheritedWidgetCount=0
  parentClass=/Script/MovieRenderPipelineSettings.MoviePipelineBurnInWidget
  rootWidgetClass=&CanvasPanel
  widgetCount=6
 }
 widgets [
  CanvasPanel_0 {widgetClassPath=&CanvasPanel, widget=CanvasPanel_0}
  Name {
   widgetClassPath=&TextBlock
   bIsVariable=true
   parent=CanvasPanel_0
   widget=Name
   properties {
    text=Name
    font {
     bForceMonospaced=false
     bMaterialIsStencil=false
     fontMaterial=null
     fontObject=&Roboto
     letterSpacing=0
     monospacedWidth=1
     size=30
     skewAmount=0
     typefaceFontName=Bold
     outlineSettings {
      bApplyOutlineToDropShadows=false
      bMiteredCorners=false
      bSeparateFillAlpha=false
      outlineMaterial=null
      outlineSize=0
      outlineColor=<rgba 0 0 0 1>
     }
    }
   }
   slot {class=CanvasPanelSlot, properties {layoutData {alignment=<0 0>, anchors {maximum=<0 0>, minimum=<0 0>}, offsets=<ltrb 20 20 800 50>}}}
  }
  Pass {
   widgetClassPath=&TextBlock
   bIsVariable=true
   parent=CanvasPanel_0
   widget=Pass
   properties {
    text=PassName
    font {
     bForceMonospaced=false
     bMaterialIsStencil=false
     fontMaterial=null
     fontObject=&Roboto
     letterSpacing=0
     monospacedWidth=1
     size=30
     skewAmount=0
     typefaceFontName=Bold
     outlineSettings {
      bApplyOutlineToDropShadows=false
      bMiteredCorners=false
      bSeparateFillAlpha=false
      outlineMaterial=null
      outlineSize=0
      outlineColor=<rgba 0 0 0 1>
     }
    }
   }
   slot {class=CanvasPanelSlot, properties {layoutData {alignment=<0 0>, anchors {maximum=<0 1>, minimum=<0 1>}, offsets=<ltrb 20 -70 500 50>}}}
  }
  frameNumber {
   widgetClassPath=&TextBlock
   bIsVariable=true
   parent=CanvasPanel_0
   widget=frameNumber
   properties {
    justification=Right
    text=9999
    font {
     bForceMonospaced=false
     bMaterialIsStencil=false
     fontMaterial=null
     fontObject=&Roboto
     letterSpacing=0
     monospacedWidth=1
     size=30
     skewAmount=0
     typefaceFontName=Bold
     outlineSettings {
      bApplyOutlineToDropShadows=false
      bMiteredCorners=false
      bSeparateFillAlpha=false
      outlineMaterial=null
      outlineSize=0
      outlineColor=<rgba 0 0 0 1>
     }
    }
   }
   slot {class=CanvasPanelSlot, properties {layoutData {alignment=<0 0>, anchors {maximum=<1 1>, minimum=<1 1>}, offsets=<ltrb -140 -70 120 50>}}}
  }
  BuildNumber {
   widgetClassPath=&TextBlock
   bIsVariable=true
   parent=CanvasPanel_0
   widget=BuildNumber
   properties {
    justification=Right
    text=Build
    font {
     bForceMonospaced=false
     bMaterialIsStencil=false
     fontMaterial=null
     fontObject=&Roboto
     letterSpacing=0
     monospacedWidth=1
     size=20
     skewAmount=0
     typefaceFontName=Bold
     outlineSettings {
      bApplyOutlineToDropShadows=false
      bMiteredCorners=false
      bSeparateFillAlpha=false
      outlineMaterial=null
      outlineSize=0
      outlineColor=<rgba 0 0 0 1>
     }
    }
   }
   slot {class=CanvasPanelSlot, properties {layoutData {alignment=<0 0>, anchors {maximum=<1 0>, minimum=<1 0>}, offsets=<ltrb -520 104 500 60.02>}}}
  }
  Date {
   widgetClassPath=&TextBlock
   bIsVariable=true
   parent=CanvasPanel_0
   widget=Date
   properties {
    justification=Right
    text=Date
    font {
     bForceMonospaced=false
     bMaterialIsStencil=false
     fontMaterial=null
     fontObject=&Roboto
     letterSpacing=0
     monospacedWidth=1
     size=30
     skewAmount=0
     typefaceFontName=Bold
     outlineSettings {
      bApplyOutlineToDropShadows=false
      bMiteredCorners=false
      bSeparateFillAlpha=false
      outlineMaterial=null
      outlineSize=0
      outlineColor=<rgba 0 0 0 1>
     }
    }
   }
   slot {class=CanvasPanelSlot, properties {layoutData {alignment=<0 0>, anchors {maximum=<1 0>, minimum=<1 0>}, offsets=<ltrb -520 20 500 50>}}}
  }
 ]
}
