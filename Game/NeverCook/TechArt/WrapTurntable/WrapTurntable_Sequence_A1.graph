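# Tech-art Level Sequence that runs a turntable for wraps, apparently to review them on weapons and vehicles, at 30 fps.
# A camera cut track steps through timed shots, with possessed weapon and vehicle cameras holding their own field of view.
# The tilted TurntableWeapons actor has keyed 360-degree yaw turns.
# Event triggers drive BP_WrapTurntable across the timeline and BP_Turntable_GBuffers in an early burst, likely for buffer views.
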LevelSequence /Game/NeverCook/TechArt/WrapTurntable/WrapTurntable_Sequence_A1
bindings [
 CameraComponent {
  id=#1
  parent_id=#2
  possessed_class=CameraComponent
  tracks [{
    type=MovieSceneFloatTrack
    display_name=Field Of View
    sections [{type=MovieSceneFloatSection, channels [None {type=MovieSceneScriptingFloatChannel, default=40}]}]
   }]
 }
 CameraWeaponsB {id=#3, possessed_class=CameraActor, tracks [{type=MovieScene3DTransformTrack, sections []}]}
 CameraComponent {
  id=#4
  parent_id=#5
  possessed_class=CameraComponent
  tracks [{
    type=MovieSceneFloatTrack
    display_name=Field Of View
    sections [{type=MovieSceneFloatSection, channels [None {type=MovieSceneScriptingFloatChannel, default=90}]}]
   }]
 }
 CameraComponent {
  id=#6
  parent_id=#3
  possessed_class=CameraComponent
  tracks [{
    type=MovieSceneFloatTrack
    display_name=Field Of View
    sections [{type=MovieSceneFloatSection, channels [None {type=MovieSceneScriptingFloatChannel, default=40}]}]
   }]
 }
 CameraWeaponsA {id=#7, possessed_class=CameraActor, tracks [{type=MovieScene3DTransformTrack, sections []}]}
 CameraWeaponsA3 {
  id=#8
  possessed_class=CameraActor
  tracks [{
    type=MovieScene3DTransformTrack
    sections [{
      type=MovieScene3DTransformSection
      channels [
       Location.X {type=MovieSceneScriptingDoubleChannel, default=4412}
       Location.Y {type=MovieSceneScriptingDoubleChannel, default=333.1476}
       Location.Z {type=MovieSceneScriptingDoubleChannel, default=-50}
       Rotation.X {type=MovieSceneScriptingDoubleChannel, default=0}
       Rotation.Y {type=MovieSceneScriptingDoubleChannel, default=0}
       Rotation.Z {type=MovieSceneScriptingDoubleChannel, default=-90}
       Scale.X {type=MovieSceneScriptingDoubleChannel, default=1}
       Scale.Y {type=MovieSceneScriptingDoubleChannel, default=1}
       Scale.Z {type=MovieSceneScriptingDoubleChannel, default=1}
      ]
     }]
   }]
 }
 CameraWeaponsA2 {
  id=#9
  possessed_class=CameraActor
  tracks [{
    type=MovieScene3DTransformTrack
    sections [{
      type=MovieScene3DTransformSection
      channels [
       Location.X {type=MovieSceneScriptingDoubleChannel, default=4056}
       Location.Y {type=MovieSceneScriptingDoubleChannel, default=333.1476}
       Location.Z {type=MovieSceneScriptingDoubleChannel, default=-50}
       Rotation.X {type=MovieSceneScriptingDoubleChannel, default=0}
       Rotation.Y {type=MovieSceneScriptingDoubleChannel, default=0}
       Rotation.Z {type=MovieSceneScriptingDoubleChannel, default=-90}
       Scale.X {type=MovieSceneScriptingDoubleChannel, default=1}
       Scale.Y {type=MovieSceneScriptingDoubleChannel, default=1}
       Scale.Z {type=MovieSceneScriptingDoubleChannel, default=1}
      ]
     }]
   }]
 }
 BP_WrapTurntable {
  id=#10
  possessed_class=BP_WrapTurntable_C
  tracks [{
    type=MovieSceneEventTrack
    sections [{
      type=MovieSceneEventTriggerSection
      channels [None {
        type=MovieSceneScriptingEventChannel
        keys [
         [10, "<Struct 'MovieSceneEvent' (0x49e5f37cee0) {}>"]
         [38, "<Struct 'MovieSceneEvent' (0x49e5f37cda0) {}>"]
         [251, "<Struct 'MovieSceneEvent' (0x49e5f37ce40) {}>"]
         [521, "<Struct 'MovieSceneEvent' (0x49e5f37cf80) {}>"]
         [567, "<Struct 'MovieSceneEvent' (0x49e5f37d200) {}>"]
         [619, "<Struct 'MovieSceneEvent' (0x49e5f37d020) {}>"]
         [661, "<Struct 'MovieSceneEvent' (0x49e5f37d980) {}>"]
         [701, "<Struct 'MovieSceneEvent' (0x49e5f37bcc0) {}>"]
         [781, "<Struct 'MovieSceneEvent' (0x49e5f37da20) {}>"]
         [861, "<Struct 'MovieSceneEvent' (0x49e5f37cb20) {}>"]
         [941, "<Struct 'MovieSceneEvent' (0x49e5f37d2a0) {}>"]
         [1021, "<Struct 'MovieSceneEvent' (0x49e5f37d660) {}>"]
        ]
       }]
     }]
   }]
 }
 CameraVehicles {id=#11, possessed_class=CameraActor, tracks [{type=MovieScene3DTransformTrack, sections []}]}
 CameraWeaponsC {
  id=#12
  possessed_class=CameraActor
  tracks [{
    type=MovieScene3DTransformTrack
    sections [{
      type=MovieScene3DTransformSection
      channels [
       Location.X {type=MovieSceneScriptingDoubleChannel, default=-616.3299}
       Location.Y {type=MovieSceneScriptingDoubleChannel, default=213.7608}
       Location.Z {type=MovieSceneScriptingDoubleChannel, default=-26.6542}
       Rotation.X {type=MovieSceneScriptingDoubleChannel, default=0}
       Rotation.Y {type=MovieSceneScriptingDoubleChannel, default=-10}
       Rotation.Z {type=MovieSceneScriptingDoubleChannel, default=-90.1999}
       Scale.X {type=MovieSceneScriptingDoubleChannel, default=1}
       Scale.Y {type=MovieSceneScriptingDoubleChannel, default=1}
       Scale.Z {type=MovieSceneScriptingDoubleChannel, default=1}
      ]
     }]
   }]
 }
 CameraWeaponsA4 {
  id=#2
  possessed_class=CameraActor
  tracks [{
    type=MovieScene3DTransformTrack
    sections [{
      type=MovieScene3DTransformSection
      channels [
       Location.X {type=MovieSceneScriptingDoubleChannel, default=4652}
       Location.Y {type=MovieSceneScriptingDoubleChannel, default=329.1483}
       Location.Z {type=MovieSceneScriptingDoubleChannel, default=-50.0005}
       Rotation.X {type=MovieSceneScriptingDoubleChannel, default=0}
       Rotation.Y {type=MovieSceneScriptingDoubleChannel, default=0}
       Rotation.Z {type=MovieSceneScriptingDoubleChannel, default=-90}
       Scale.X {type=MovieSceneScriptingDoubleChannel, default=1}
       Scale.Y {type=MovieSceneScriptingDoubleChannel, default=1}
       Scale.Z {type=MovieSceneScriptingDoubleChannel, default=1}
      ]
     }]
   }]
 }
 CameraComponent {
  id=#13
  parent_id=#8
  possessed_class=CameraComponent
  tracks [{
    type=MovieSceneFloatTrack
    display_name=Field Of View
    sections [{type=MovieSceneFloatSection, channels [None {type=MovieSceneScriptingFloatChannel, default=40}]}]
   }]
 }
 WeaponsB {id=#14, possessed_class=Actor, tracks []}
 CameraComponent {
  id=#15
  parent_id=#9
  possessed_class=CameraComponent
  tracks [{
    type=MovieSceneFloatTrack
    display_name=Field Of View
    sections [{type=MovieSceneFloatSection, channels [None {type=MovieSceneScriptingFloatChannel, default=40}]}]
   }]
 }
 BP_Turntable_GBuffers {
  id=#16
  possessed_class=BP_Turntable_GBuffers_Wraps_C
  tracks [{
    type=MovieSceneEventTrack
    sections [{
      type=MovieSceneEventTriggerSection
      channels [None {
        type=MovieSceneScriptingEventChannel
        keys [
         [10, "<Struct 'MovieSceneEvent' (0x49e5f37d520) {}>"]
         [17, "<Struct 'MovieSceneEvent' (0x49e5f37d5c0) {}>"]
         [23, "<Struct 'MovieSceneEvent' (0x49e5f37d840) {}>"]
         [29, "<Struct 'MovieSceneEvent' (0x49e5f37dac0) {}>"]
         [35, "<Struct 'MovieSceneEvent' (0x49e5f37db60) {}>"]
         [42, "<Struct 'MovieSceneEvent' (0x49e5f37dc00) {}>"]
         [49, "<Struct 'MovieSceneEvent' (0x49e5f37dca0) {}>"]
        ]
       }]
     }]
   }]
 }
 CameraComponent {
  id=#17
  parent_id=#7
  possessed_class=CameraComponent
  tracks [{
    type=MovieSceneFloatTrack
    display_name=Field Of View
    sections [{type=MovieSceneFloatSection, channels [None {type=MovieSceneScriptingFloatChannel, default=40}]}]
   }]
 }
 CameraWeaponsC2 {
  id=#5
  possessed_class=CameraActor
  tracks [{
    type=MovieScene3DTransformTrack
    sections [{
      type=MovieScene3DTransformSection
      channels [
       Location.X {type=MovieSceneScriptingDoubleChannel, default=-616.3299}
       Location.Y {type=MovieSceneScriptingDoubleChannel, default=213.7608}
       Location.Z {type=MovieSceneScriptingDoubleChannel, default=-26.6542}
       Rotation.X {type=MovieSceneScriptingDoubleChannel, default=0}
       Rotation.Y {type=MovieSceneScriptingDoubleChannel, default=-10}
       Rotation.Z {type=MovieSceneScriptingDoubleChannel, default=-90.1999}
       Scale.X {type=MovieSceneScriptingDoubleChannel, default=1}
       Scale.Y {type=MovieSceneScriptingDoubleChannel, default=1}
       Scale.Z {type=MovieSceneScriptingDoubleChannel, default=1}
      ]
     }]
   }]
 }
 CameraComponent {
  id=#18
  parent_id=#12
  possessed_class=CameraComponent
  tracks [{
    type=MovieSceneFloatTrack
    display_name=Field Of View
    sections [{type=MovieSceneFloatSection, channels [None {type=MovieSceneScriptingFloatChannel, default=90}]}]
   }]
 }
 TurntableWeapons {
  id=#19
  possessed_class=Actor
  tracks [{
    type=MovieScene3DTransformTrack
    sections [{
      type=MovieScene3DTransformSection
      channels [
       Location.X {type=MovieSceneScriptingDoubleChannel, default=0}
       Location.Y {type=MovieSceneScriptingDoubleChannel, default=0}
       Location.Z {type=MovieSceneScriptingDoubleChannel, default=0}
       Rotation.X {type=MovieSceneScriptingDoubleChannel, default=0}
       Rotation.Y {type=MovieSceneScriptingDoubleChannel, default=15}
       Rotation.Z {
        type=MovieSceneScriptingDoubleChannel
        default=0
        keys [[50,0],[100,360],[710,0],[750,360],[790,0],[830,360],[870,0],[910,360],[950,0],[990,360],[1030,0],[1070,360]]
       }
       Scale.X {type=MovieSceneScriptingDoubleChannel, default=1}
       Scale.Y {type=MovieSceneScriptingDoubleChannel, default=1}
       Scale.Z {type=MovieSceneScriptingDoubleChannel, default=1}
      ]
     }]
   }]
 }
 CameraComponent {
  id=#20
  parent_id=#11
  possessed_class=CameraComponent
  tracks [{
    type=MovieSceneFloatTrack
    display_name=Field Of View
    sections [{type=MovieSceneFloatSection, channels [None {type=MovieSceneScriptingFloatChannel, default=40}]}]
   }]
 }
 WeaponsA {id=#21, possessed_class=Actor, tracks []}
]
display_rate {denominator=1, numerator=30}
tick_resolution {denominator=1, numerator=24000}
tracks [{
  type=MovieSceneCameraCutTrack
  sections [
   {type=MovieSceneCameraCutSection, end_frame=42, start_frame=-12}
   {type=MovieSceneCameraCutSection, end_frame=49, start_frame=42}
   {type=MovieSceneCameraCutSection, end_frame=100, start_frame=49}
   {type=MovieSceneCameraCutSection, end_frame=200, start_frame=100}
   {type=MovieSceneCameraCutSection, end_frame=260, start_frame=200}
   {type=MovieSceneCameraCutSection, end_frame=320, start_frame=260}
   {type=MovieSceneCameraCutSection, end_frame=420, start_frame=320}
   {type=MovieSceneCameraCutSection, end_frame=470, start_frame=420}
   {type=MovieSceneCameraCutSection, end_frame=530, start_frame=470}
   {type=MovieSceneCameraCutSection, end_frame=576, start_frame=530}
   {type=MovieSceneCameraCutSection, end_frame=628, start_frame=576}
   {type=MovieSceneCameraCutSection, end_frame=670, start_frame=628}
   {type=MovieSceneCameraCutSection, end_frame=710, start_frame=670}
   {type=MovieSceneCameraCutSection, end_frame=750, start_frame=710}
   {type=MovieSceneCameraCutSection, end_frame=790, start_frame=750}
   {type=MovieSceneCameraCutSection, end_frame=830, start_frame=790}
   {type=MovieSceneCameraCutSection, end_frame=870, start_frame=830}
   {type=MovieSceneCameraCutSection, end_frame=910, start_frame=870}
   {type=MovieSceneCameraCutSection, end_frame=950, start_frame=910}
   {type=MovieSceneCameraCutSection, end_frame=990, start_frame=950}
   {type=MovieSceneCameraCutSection, end_frame=1030, start_frame=990}
   {type=MovieSceneCameraCutSection, end_frame=1070, start_frame=1030}
   {type=MovieSceneCameraCutSection, end_frame=1110, start_frame=1070}
  ]
 }]
