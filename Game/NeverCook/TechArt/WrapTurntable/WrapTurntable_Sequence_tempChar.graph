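# Level Sequence for a tech-art wrap turntable, apparently used to preview wraps on weapons and vehicles, with "tempChar" in its name.
# A camera cut track steps through six roughly 100-frame shots at 30 fps, with six bound cameras at 40 FOV.
# TurntableWeapons is tilted and keyed through alternating full yaw turns every 100 frames, continuing beyond the last cut.
# Event keys on BP_WrapTurntable fire at the start and just before the first three cut changes.
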LevelSequence /Game/NeverCook/TechArt/WrapTurntable/WrapTurntable_Sequence_tempChar
bindings [
 CameraComponent {
  id=#1
  parent_id=#2
  possessed_class=CameraComponent
  tracks [{
    type=MovieSceneFloatTrack
    display_name=Field Of View
    sections [{type=MovieSceneFloatSection, channels [None {type=MovieSceneScriptingFloatChannel, default=40}]}]
   }]
 }
 CameraWeaponsB {id=#3, possessed_class=CameraActor, tracks [{type=MovieScene3DTransformTrack, sections []}]}
 CameraComponent {
  id=#4
  parent_id=#3
  possessed_class=CameraComponent
  tracks [{
    type=MovieSceneFloatTrack
    display_name=Field Of View
    sections [{type=MovieSceneFloatSection, channels [None {type=MovieSceneScriptingFloatChannel, default=40}]}]
   }]
 }
 CameraWeaponsA {id=#5, possessed_class=CameraActor, tracks [{type=MovieScene3DTransformTrack, sections []}]}
 CameraWeaponsA3 {
  id=#6
  possessed_class=CameraActor
  tracks [{
    type=MovieScene3DTransformTrack
    sections [{
      type=MovieScene3DTransformSection
      channels [
       Location.X {type=MovieSceneScriptingDoubleChannel, default=4412}
       Location.Y {type=MovieSceneScriptingDoubleChannel, default=333.1476}
       Location.Z {type=MovieSceneScriptingDoubleChannel, default=-50}
       Rotation.X {type=MovieSceneScriptingDoubleChannel, default=0}
       Rotation.Y {type=MovieSceneScriptingDoubleChannel, default=0}
       Rotation.Z {type=MovieSceneScriptingDoubleChannel, default=-90}
       Scale.X {type=MovieSceneScriptingDoubleChannel, default=1}
       Scale.Y {type=MovieSceneScriptingDoubleChannel, default=1}
       Scale.Z {type=MovieSceneScriptingDoubleChannel, default=1}
      ]
     }]
   }]
 }
 CameraWeaponsA2 {
  id=#7
  possessed_class=CameraActor
  tracks [{
    type=MovieScene3DTransformTrack
    sections [{
      type=MovieScene3DTransformSection
      channels [
       Location.X {type=MovieSceneScriptingDoubleChannel, default=4056}
       Location.Y {type=MovieSceneScriptingDoubleChannel, default=333.1476}
       Location.Z {type=MovieSceneScriptingDoubleChannel, default=-50}
       Rotation.X {type=MovieSceneScriptingDoubleChannel, default=0}
       Rotation.Y {type=MovieSceneScriptingDoubleChannel, default=0}
       Rotation.Z {type=MovieSceneScriptingDoubleChannel, default=-90}
       Scale.X {type=MovieSceneScriptingDoubleChannel, default=1}
       Scale.Y {type=MovieSceneScriptingDoubleChannel, default=1}
       Scale.Z {type=MovieSceneScriptingDoubleChannel, default=1}
      ]
     }]
   }]
 }
 BP_WrapTurntable {
  id=#8
  possessed_class=BP_WrapTurntable_C
  tracks [{
    type=MovieSceneEventTrack
    sections [{
      type=MovieSceneEventTriggerSection
      channels [None {
        type=MovieSceneScriptingEventChannel
        keys [
         [0, "<Struct 'MovieSceneEvent' (0x49e5f37e240) {}>"]
         [0, "<Struct 'MovieSceneEvent' (0x49e5f37cda0) {}>"]
         [97, "<Struct 'MovieSceneEvent' (0x49e5f37dde0) {}>"]
         [197, "<Struct 'MovieSceneEvent' (0x49e5f37dfc0) {}>"]
         [297, "<Struct 'MovieSceneEvent' (0x49e5f37e060) {}>"]
        ]
       }]
     }]
   }]
 }
 CameraVehicles {id=#9, possessed_class=CameraActor, tracks [{type=MovieScene3DTransformTrack, sections []}]}
 CameraWeaponsA4 {
  id=#2
  possessed_class=CameraActor
  tracks [{
    type=MovieScene3DTransformTrack
    sections [{
      type=MovieScene3DTransformSection
      channels [
       Location.X {type=MovieSceneScriptingDoubleChannel, default=4652}
       Location.Y {type=MovieSceneScriptingDoubleChannel, default=329.1483}
       Location.Z {type=MovieSceneScriptingDoubleChannel, default=-50.0005}
       Rotation.X {type=MovieSceneScriptingDoubleChannel, default=0}
       Rotation.Y {type=MovieSceneScriptingDoubleChannel, default=0}
       Rotation.Z {type=MovieSceneScriptingDoubleChannel, default=-90}
       Scale.X {type=MovieSceneScriptingDoubleChannel, default=1}
       Scale.Y {type=MovieSceneScriptingDoubleChannel, default=1}
       Scale.Z {type=MovieSceneScriptingDoubleChannel, default=1}
      ]
     }]
   }]
 }
 CameraComponent {
  id=#10
  parent_id=#6
  possessed_class=CameraComponent
  tracks [{
    type=MovieSceneFloatTrack
    display_name=Field Of View
    sections [{type=MovieSceneFloatSection, channels [None {type=MovieSceneScriptingFloatChannel, default=40}]}]
   }]
 }
 WeaponsB {id=#11, possessed_class=Actor, tracks []}
 CameraComponent {
  id=#12
  parent_id=#7
  possessed_class=CameraComponent
  tracks [{
    type=MovieSceneFloatTrack
    display_name=Field Of View
    sections [{type=MovieSceneFloatSection, channels [None {type=MovieSceneScriptingFloatChannel, default=40}]}]
   }]
 }
 CameraComponent {
  id=#13
  parent_id=#5
  possessed_class=CameraComponent
  tracks [{
    type=MovieSceneFloatTrack
    display_name=Field Of View
    sections [{type=MovieSceneFloatSection, channels [None {type=MovieSceneScriptingFloatChannel, default=40}]}]
   }]
 }
 TurntableWeapons {
  id=#14
  possessed_class=Actor
  tracks [{
    type=MovieScene3DTransformTrack
    sections [{
      type=MovieScene3DTransformSection
      channels [
       Location.X {type=MovieSceneScriptingDoubleChannel, default=0}
       Location.Y {type=MovieSceneScriptingDoubleChannel, default=0}
       Location.Z {type=MovieSceneScriptingDoubleChannel, default=0}
       Rotation.X {type=MovieSceneScriptingDoubleChannel, default=0}
       Rotation.Y {type=MovieSceneScriptingDoubleChannel, default=15}
       Rotation.Z {
        type=MovieSceneScriptingDoubleChannel
        default=0
        keys [[0,0],[100,360],[200,0],[300,360],[400,0],[500,360],[600,0],[700,360],[800,0],[900,360],[1000,0],[1100,360]]
       }
       Scale.X {type=MovieSceneScriptingDoubleChannel, default=1}
       Scale.Y {type=MovieSceneScriptingDoubleChannel, default=1}
       Scale.Z {type=MovieSceneScriptingDoubleChannel, default=1}
      ]
     }]
   }]
 }
 CameraComponent {
  id=#15
  parent_id=#9
  possessed_class=CameraComponent
  tracks [{
    type=MovieSceneFloatTrack
    display_name=Field Of View
    sections [{type=MovieSceneFloatSection, channels [None {type=MovieSceneScriptingFloatChannel, default=40}]}]
   }]
 }
 WeaponsA {id=#16, possessed_class=Actor, tracks []}
]
display_rate {denominator=1, numerator=30}
tick_resolution {denominator=1, numerator=24000}
tracks [{
  type=MovieSceneCameraCutTrack
  sections [
   {type=MovieSceneCameraCutSection, end_frame=99, start_frame=0}
   {type=MovieSceneCameraCutSection, end_frame=199, start_frame=99}
   {type=MovieSceneCameraCutSection, end_frame=299, start_frame=199}
   {type=MovieSceneCameraCutSection, end_frame=399, start_frame=299}
   {type=MovieSceneCameraCutSection, end_frame=499, start_frame=399}
   {type=MovieSceneCameraCutSection, end_frame=599, start_frame=499}
  ]
 }]
