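# Level Sequence in TechArt that drives a turntable, apparently for showing wraps on weapons and vehicles; its name marks it as temporary.
# TurntableWeapons is tilted and keys yaw between 0 and 360 degrees through roughly the first half.
# An event track triggers BP_WrapTurntable at intervals throughout; the called functions are not shown.
# A camera cut track splits the timeline into short shots, alongside bound weapon and vehicle cameras with fixed FOVs.
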LevelSequence /Game/NeverCook/TechArt/WrapTurntable/WrapTurntable_Sequence_temp_Ref_Half
bindings [
 CameraComponent {
  id=#1
  parent_id=#2
  possessed_class=CameraComponent
  tracks [{
    type=MovieSceneFloatTrack
    display_name=Field Of View
    sections [{type=MovieSceneFloatSection, channels [None {type=MovieSceneScriptingFloatChannel, default=40}]}]
   }]
 }
 CameraWeaponsB {id=#3, possessed_class=CameraActor, tracks [{type=MovieScene3DTransformTrack, sections []}]}
 CameraComponent {
  id=#4
  parent_id=#3
  possessed_class=CameraComponent
  tracks [{
    type=MovieSceneFloatTrack
    display_name=Field Of View
    sections [{type=MovieSceneFloatSection, channels [None {type=MovieSceneScriptingFloatChannel, default=40}]}]
   }]
 }
 CameraWeaponsA {id=#5, possessed_class=CameraActor, tracks [{type=MovieScene3DTransformTrack, sections []}]}
 CameraWeaponsA3 {
  id=#6
  possessed_class=CameraActor
  tracks [{
    type=MovieScene3DTransformTrack
    sections [{
      type=MovieScene3DTransformSection
      channels [
       Location.X {type=MovieSceneScriptingDoubleChannel, default=4412}
       Location.Y {type=MovieSceneScriptingDoubleChannel, default=333.1476}
       Location.Z {type=MovieSceneScriptingDoubleChannel, default=-50}
       Rotation.X {type=MovieSceneScriptingDoubleChannel, default=0}
       Rotation.Y {type=MovieSceneScriptingDoubleChannel, default=0}
       Rotation.Z {type=MovieSceneScriptingDoubleChannel, default=-90}
       Scale.X {type=MovieSceneScriptingDoubleChannel, default=1}
       Scale.Y {type=MovieSceneScriptingDoubleChannel, default=1}
       Scale.Z {type=MovieSceneScriptingDoubleChannel, default=1}
      ]
     }]
   }]
 }
 CameraWeaponsA2 {
  id=#7
  possessed_class=CameraActor
  tracks [{
    type=MovieScene3DTransformTrack
    sections [{
      type=MovieScene3DTransformSection
      channels [
       Location.X {type=MovieSceneScriptingDoubleChannel, default=4056}
       Location.Y {type=MovieSceneScriptingDoubleChannel, default=333.1476}
       Location.Z {type=MovieSceneScriptingDoubleChannel, default=-50}
       Rotation.X {type=MovieSceneScriptingDoubleChannel, default=0}
       Rotation.Y {type=MovieSceneScriptingDoubleChannel, default=0}
       Rotation.Z {type=MovieSceneScriptingDoubleChannel, default=-90}
       Scale.X {type=MovieSceneScriptingDoubleChannel, default=1}
       Scale.Y {type=MovieSceneScriptingDoubleChannel, default=1}
       Scale.Z {type=MovieSceneScriptingDoubleChannel, default=1}
      ]
     }]
   }]
 }
 BP_WrapTurntable {
  id=#8
  possessed_class=BP_WrapTurntable_C
  tracks [{
    type=MovieSceneEventTrack
    sections [{
      type=MovieSceneEventTriggerSection
      channels [None {
        type=MovieSceneScriptingEventChannel
        keys [
         [10, "<Struct 'MovieSceneEvent' (0x49e5f37e6a0) {}>"]
         [161, "<Struct 'MovieSceneEvent' (0x49e5f37e240) {}>"]
         [261, "<Struct 'MovieSceneEvent' (0x49e5f37e060) {}>"]
         [361, "<Struct 'MovieSceneEvent' (0x49e5f37dde0) {}>"]
         [461, "<Struct 'MovieSceneEvent' (0x49e5f37e7e0) {}>"]
         [561, "<Struct 'MovieSceneEvent' (0x49e5f37dfc0) {}>"]
         [661, "<Struct 'MovieSceneEvent' (0x49e5f37d660) {}>"]
         [711, "<Struct 'MovieSceneEvent' (0x49e5f37e560) {}>"]
         [761, "<Struct 'MovieSceneEvent' (0x49e5f37d5c0) {}>"]
         [811, "<Struct 'MovieSceneEvent' (0x49e5f37e740) {}>"]
         [861, "<Struct 'MovieSceneEvent' (0x49e5f37d520) {}>"]
         [911, "<Struct 'MovieSceneEvent' (0x49e5f37da20) {}>"]
         [962, "<Struct 'MovieSceneEvent' (0x49e5f37dca0) {}>"]
        ]
       }]
     }]
   }]
 }
 CameraVehicles {id=#9, possessed_class=CameraActor, tracks [{type=MovieScene3DTransformTrack, sections []}]}
 CameraWeaponsC {
  id=#10
  possessed_class=CameraActor
  tracks [{
    type=MovieScene3DTransformTrack
    sections [{
      type=MovieScene3DTransformSection
      channels [
       Location.X {type=MovieSceneScriptingDoubleChannel, default=-616.3299}
       Location.Y {type=MovieSceneScriptingDoubleChannel, default=213.7608}
       Location.Z {type=MovieSceneScriptingDoubleChannel, default=-26.6542}
       Rotation.X {type=MovieSceneScriptingDoubleChannel, default=0}
       Rotation.Y {type=MovieSceneScriptingDoubleChannel, default=-10}
       Rotation.Z {type=MovieSceneScriptingDoubleChannel, default=-90.1999}
       Scale.X {type=MovieSceneScriptingDoubleChannel, default=1}
       Scale.Y {type=MovieSceneScriptingDoubleChannel, default=1}
       Scale.Z {type=MovieSceneScriptingDoubleChannel, default=1}
      ]
     }]
   }]
 }
 CameraWeaponsA4 {
  id=#2
  possessed_class=CameraActor
  tracks [{
    type=MovieScene3DTransformTrack
    sections [{
      type=MovieScene3DTransformSection
      channels [
       Location.X {type=MovieSceneScriptingDoubleChannel, default=4652}
       Location.Y {type=MovieSceneScriptingDoubleChannel, default=329.1483}
       Location.Z {type=MovieSceneScriptingDoubleChannel, default=-50.0005}
       Rotation.X {type=MovieSceneScriptingDoubleChannel, default=0}
       Rotation.Y {type=MovieSceneScriptingDoubleChannel, default=0}
       Rotation.Z {type=MovieSceneScriptingDoubleChannel, default=-90}
       Scale.X {type=MovieSceneScriptingDoubleChannel, default=1}
       Scale.Y {type=MovieSceneScriptingDoubleChannel, default=1}
       Scale.Z {type=MovieSceneScriptingDoubleChannel, default=1}
      ]
     }]
   }]
 }
 CameraComponent {
  id=#11
  parent_id=#6
  possessed_class=CameraComponent
  tracks [{
    type=MovieSceneFloatTrack
    display_name=Field Of View
    sections [{type=MovieSceneFloatSection, channels [None {type=MovieSceneScriptingFloatChannel, default=40}]}]
   }]
 }
 WeaponsB {id=#12, possessed_class=Actor, tracks []}
 CameraComponent {
  id=#13
  parent_id=#7
  possessed_class=CameraComponent
  tracks [{
    type=MovieSceneFloatTrack
    display_name=Field Of View
    sections [{type=MovieSceneFloatSection, channels [None {type=MovieSceneScriptingFloatChannel, default=40}]}]
   }]
 }
 CameraComponent {
  id=#14
  parent_id=#5
  possessed_class=CameraComponent
  tracks [{
    type=MovieSceneFloatTrack
    display_name=Field Of View
    sections [{type=MovieSceneFloatSection, channels [None {type=MovieSceneScriptingFloatChannel, default=40}]}]
   }]
 }
 CameraComponent {
  id=#15
  parent_id=#10
  possessed_class=CameraComponent
  tracks [{
    type=MovieSceneFloatTrack
    display_name=Field Of View
    sections [{type=MovieSceneFloatSection, channels [None {type=MovieSceneScriptingFloatChannel, default=90}]}]
   }]
 }
 TurntableWeapons {
  id=#16
  possessed_class=Actor
  tracks [{
    type=MovieScene3DTransformTrack
    sections [{
      type=MovieScene3DTransformSection
      channels [
       Location.X {type=MovieSceneScriptingDoubleChannel, default=0}
       Location.Y {type=MovieSceneScriptingDoubleChannel, default=0}
       Location.Z {type=MovieSceneScriptingDoubleChannel, default=0}
       Rotation.X {type=MovieSceneScriptingDoubleChannel, default=0}
       Rotation.Y {type=MovieSceneScriptingDoubleChannel, default=15}
       Rotation.Z {
        type=MovieSceneScriptingDoubleChannel
        default=0
        keys [[10,0],[60,360],[160,0],[210,360],[260,0],[310,360],[360,0],[410,360],[460,0],[510,360],[560,0],[610,360]]
       }
       Scale.X {type=MovieSceneScriptingDoubleChannel, default=1}
       Scale.Y {type=MovieSceneScriptingDoubleChannel, default=1}
       Scale.Z {type=MovieSceneScriptingDoubleChannel, default=1}
      ]
     }]
   }]
 }
 CameraComponent {
  id=#17
  parent_id=#9
  possessed_class=CameraComponent
  tracks [{
    type=MovieSceneFloatTrack
    display_name=Field Of View
    sections [{type=MovieSceneFloatSection, channels [None {type=MovieSceneScriptingFloatChannel, default=40}]}]
   }]
 }
 WeaponsA {id=#18, possessed_class=Actor, tracks []}
]
display_rate {denominator=1, numerator=30}
tick_resolution {denominator=1, numerator=24000}
tracks [{
  type=MovieSceneCameraCutTrack
  sections [
   {type=MovieSceneCameraCutSection, end_frame=10, start_frame=0}
   {type=MovieSceneCameraCutSection, end_frame=60, start_frame=10}
   {type=MovieSceneCameraCutSection, end_frame=160, start_frame=60}
   {type=MovieSceneCameraCutSection, end_frame=210, start_frame=160}
   {type=MovieSceneCameraCutSection, end_frame=260, start_frame=210}
   {type=MovieSceneCameraCutSection, end_frame=310, start_frame=260}
   {type=MovieSceneCameraCutSection, end_frame=360, start_frame=310}
   {type=MovieSceneCameraCutSection, end_frame=410, start_frame=360}
   {type=MovieSceneCameraCutSection, end_frame=460, start_frame=410}
   {type=MovieSceneCameraCutSection, end_frame=510, start_frame=460}
   {type=MovieSceneCameraCutSection, end_frame=560, start_frame=510}
   {type=MovieSceneCameraCutSection, end_frame=610, start_frame=560}
   {type=MovieSceneCameraCutSection, end_frame=660, start_frame=610}
   {type=MovieSceneCameraCutSection, end_frame=710, start_frame=660}
   {type=MovieSceneCameraCutSection, end_frame=760, start_frame=710}
   {type=MovieSceneCameraCutSection, end_frame=810, start_frame=760}
   {type=MovieSceneCameraCutSection, end_frame=860, start_frame=810}
   {type=MovieSceneCameraCutSection, end_frame=910, start_frame=860}
   {type=MovieSceneCameraCutSection, end_frame=960, start_frame=910}
   {type=MovieSceneCameraCutSection, end_frame=1010, start_frame=960}
   {type=MovieSceneCameraCutSection, end_frame=1060, start_frame=1010}
   {type=MovieSceneCameraCutSection, end_frame=1110, start_frame=1060}
  ]
 }]
