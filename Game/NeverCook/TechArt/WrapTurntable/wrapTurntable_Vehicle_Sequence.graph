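# Level Sequence in the TechArt WrapTurntable folder that frames a fixed camera shot, apparently for vehicle wrap turntable captures.
# It possesses the CameraVehicles actor with constant transform values, placing it elevated, offset along Y and pitched down, and sets a constant field of view on its CameraComponent.
# A single camera cut section spans about seven seconds at 30 fps.
# The sequence binds no vehicle or turntable, so it animates no rotation itself.
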LevelSequence /Game/NeverCook/TechArt/WrapTurntable/wrapTurntable_Vehicle_Sequence
bindings [
 CameraVehicles {
  id=#1
  possessed_class=CameraActor
  tracks [{
    type=MovieScene3DTransformTrack
    sections [{
      type=MovieScene3DTransformSection
      channels [
       Location.X {type=MovieSceneScriptingDoubleChannel, default=0}
       Location.Y {type=MovieSceneScriptingDoubleChannel, default=2000}
       Location.Z {type=MovieSceneScriptingDoubleChannel, default=1200}
       Rotation.X {type=MovieSceneScriptingDoubleChannel, default=0}
       Rotation.Y {type=MovieSceneScriptingDoubleChannel, default=-28}
       Rotation.Z {type=MovieSceneScriptingDoubleChannel, default=-90}
       Scale.X {type=MovieSceneScriptingDoubleChannel, default=1}
       Scale.Y {type=MovieSceneScriptingDoubleChannel, default=1}
       Scale.Z {type=MovieSceneScriptingDoubleChannel, default=1}
      ]
     }]
   }]
 }
 CameraComponent {
  id=#2
  parent_id=#1
  possessed_class=CameraComponent
  tracks [{
    type=MovieSceneFloatTrack
    display_name=Field Of View
    sections [{type=MovieSceneFloatSection, channels [None {type=MovieSceneScriptingFloatChannel, default=40}]}]
   }]
 }
]
display_rate {denominator=1, numerator=30}
tick_resolution {denominator=1, numerator=24000}
tracks [{type=MovieSceneCameraCutTrack, sections [{type=MovieSceneCameraCutSection, end_frame=210, start_frame=-2}]}]
